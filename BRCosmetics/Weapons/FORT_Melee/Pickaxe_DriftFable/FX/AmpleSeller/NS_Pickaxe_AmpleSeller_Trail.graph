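# Niagara trail effect for the AmpleSeller pickaxe, drawing ribbon trails along the weapon's swings.
# Three ribbon emitters, a main trail plus close and far black cross-hatching layers, each spawn two anim trails between socket pairs chosen by the swing section index.
# Each feeds dynamic material parameters and fades alpha over short lifetimes.
# The main trail also tints by a time-of-day color, with User.Variant blending in night-time alpha dimming.
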NiagaraSystem /BRCosmetics/Weapons/FORT_Melee/Pickaxe_DriftFable/FX/AmpleSeller/NS_Pickaxe_AmpleSeller_Trail
dependencies {
 usedDataInterfaces [
  &NiagaraDataInterfaceArrayInt32
  &NiagaraDataInterfaceCurve
  &NiagaraDataInterfaceColorCurve
  &NiagaraDataInterfaceSocketReader
 ]
 usedDynamicInputs [
  &LerpLinearColors
  &ColorFromCurve
  &SetBoolByFloatComparison
 ]
 usedModules [
  &self:SystemSpawnScript.SystemScriptSource.SystemScriptGraph.NiagaraNodeAssignment_5.SetVariables_#1
  &self:SystemSpawnScript.SystemScriptSource.SystemScriptGraph.NiagaraNodeAssignment_6.SetVariables_#2
  &SystemState
  &self:SystemSpawnScript.SystemScriptSource.SystemScriptGraph.NiagaraNodeAssignment_1.SetVariables_#3
  &self:SystemSpawnScript.SystemScriptSource.SystemScriptGraph.NiagaraNodeAssignment_0.SetVariables_#4
  &self:SystemSpawnScript.SystemScriptSource.SystemScriptGraph.NiagaraNodeAssignment_4.SetVariables_#5
  &self:SystemSpawnScript.SystemScriptSource.SystemScriptGraph.NiagaraNodeAssignment_7.SetVariables_#6
  &self:SystemSpawnScript.SystemScriptSource.SystemScriptGraph.NiagaraNodeAssignment_8.SetVariables_#7
 ]
}
emitters {
 Close_CrossHatching {
  emitterSpawnScript [&FN_RibbonIDOffsetUtility()]
  emitterUpdateScript [
   &EmitterState()
   &FN_SpawnAnimTrail(Number of Anim Trails=2)
  ]
  particleSpawnScript [
   &FN_InitializeAnimTrailBoneIndecies(
    First Bone or Socket Index:NiagaraInt32=&SelectIntFromArray(Direct Array Index=$User.SwingSectionIndex, Int Selection Array=DataInterface() {propertyValues {IntData [3,1,3,1]}})
    Second Bone or Socket Index:NiagaraInt32=&SelectIntFromArray(Direct Array Index=$User.SwingSectionIndex, Int Selection Array=DataInterface() {propertyValues {IntData [2,0,2,0]}})
   ) {enabled=false}
   &FN_InitializeAnimTrailBoneIndecies(
    First Bone or Socket Index:NiagaraInt32=&SelectIntFromArray(Direct Array Index=$User.SwingSectionIndex, Int Selection Array=DataInterface() {propertyValues {IntData [5,4,5,4]}})
    Second Bone or Socket Index:NiagaraInt32=&SelectIntFromArray(Direct Array Index=$User.SwingSectionIndex, Int Selection Array=DataInterface() {propertyValues {IntData [3,1,3,1]}})
   ) {module=FN_InitializeAnimTrailBoneIndecies002}
   &FN_InitializeAnimTrailBoneIndecies(
    AnimTrail Selection=1
    First Bone or Socket Index:NiagaraInt32=&SelectIntFromArray(Direct Array Index=$User.SwingSectionIndex, Int Selection Array=DataInterface() {propertyValues {IntData [5,4,5,4]}})
    Second Bone or Socket Index:NiagaraInt32=&SelectIntFromArray(Direct Array Index=$User.SwingSectionIndex, Int Selection Array=DataInterface() {propertyValues {IntData [2,0,2,0]}})
   ) {module=FN_InitializeAnimTrailBoneIndecies003}
   &FN_InitializeAnimTrailBoneIndecies(AnimTrail Selection=1, First Bone or Socket Index=2, Second Bone or Socket Index=3) {enabled=false, module=FN_InitializeAnimTrailBoneIndecies001}
   &FN_InitializeAnimTrail(Color=<rgba 0 0 0 1>, Lifetime=0.2)
   &AddVelocity(Velocity=<0 400 0>) {enabled=false}
   /SetVariables(
    Particles.Ribbon2Width:NiagaraFloat=&Multiply_Float(A=$Particles.RibbonWidth, B=&MaskFloatBySpawnGroup(Masked Float=1, Passthrough Float=0))
    Particles.RibbonWidth=&Multiply_Float(A=$Particles.RibbonWidth, B=&MaskFloatBySpawnGroup(Masked Float=0, Passthrough Float=1))
   ) {isSetParameters=true}
  ]
  particleUpdateScript [
   &SolveForcesAndVelocity(Clamp Velocity=false, Limit Acceleration=false)
   &FN_UpdateAnimTrail() {module=FN_UpdateAnimTrail001}
   &DynamicMaterialParameters(
    Index 1 Param 3=&FloatFromCurve(
     FloatCurve=DataInterface() {propertyValues {Curve {keys [<at 0 0>, {arriveTangent=1.9814, tangentMode=RCTM_User, time=1, value=1}]}}}
     CurveIndex=&ModuloFloat(A=$Emitter.Age, B=1)
     Scale Curve=0
    )
    Index 1 Param 4=1
    Index 2 Param 1=10
    Index 2 Param 2=1
    Index 2 Param 3=0.5
    Index 2 Param 4=0.6
    Module.Index0Float0Write=true
    Module.Index0Float1Write=true
    Module.Index0Float2Write=true
    Module.Index0Float3Write=true
    Module.Index1Float0Write=true
    Module.Index1Float1Write=true
    Module.Index1Float2Write=true
    Module.Index1Float3Write=true
    Module.Index2Float0Write=true
    Module.Index2Float1Write=true
    Module.Index2Float2Write=true
    Module.Index2Float3Write=true
    Write Parameter Index 1=::NewEnumerator1
    Write Parameter Index 2=::NewEnumerator1
   )
   &ScaleColor(
    Scale Alpha=&MaskFloatBySpawnGroup(
     Masked Float=&FloatFromCurve(FloatCurve=DataInterface() {propertyValues {Curve {keys [{interpMode=RCIM_Cubic, tangentMode=RCTM_User, time=0.1, value=1}, {arriveTangent=-2.9197, tangentMode=RCTM_User, time=0.2, value=0}]}}})
     Passthrough Float=&FloatFromCurve(FloatCurve=DataInterface() {propertyValues {Curve {keys [{interpMode=RCIM_Cubic, tangentMode=RCTM_User, time=0, value=1}, {arriveTangent=-2.9197, tangentMode=RCTM_User, time=0.15, value=0}]}}})
     Spawn Group Mask=1
    )
    Scale RGB=&VectorFromFloat(Value=&FloatFromCurve(FloatCurve=DataInterface() {propertyValues {Curve {keys [<at 0 1>,<at 1 1>]}}}))
    ScaleA=true
    ScaleRGB=true
   )
  ]
  rendererClasses [&NiagaraRibbonRendererProperties]
  renderers [
   {rendererClass=&NiagaraRibbonRendererProperties}
   {rendererClass=&NiagaraRibbonRendererProperties, rendererIndex=1}
   {rendererClass=&NiagaraRibbonRendererProperties, rendererIndex=2}
   {rendererClass=&NiagaraRibbonRendererProperties, rendererIndex=3}
  ]
 }
 Far_CrossHatching {
  emitterSpawnScript [&FN_RibbonIDOffsetUtility()]
  emitterUpdateScript [
   &EmitterState()
   &FN_SpawnAnimTrail(Number of Anim Trails=2)
  ]
  particleSpawnScript [
   &FN_InitializeAnimTrailBoneIndecies(
    First Bone or Socket Index:NiagaraInt32=&SelectIntFromArray(Direct Array Index=$User.SwingSectionIndex, Int Selection Array=DataInterface() {propertyValues {IntData [3,1,3,1]}})
    Second Bone or Socket Index:NiagaraInt32=&SelectIntFromArray(Direct Array Index=$User.SwingSectionIndex, Int Selection Array=DataInterface() {propertyValues {IntData [2,0,2,0]}})
   ) {enabled=false}
   &FN_InitializeAnimTrailBoneIndecies(
    First Bone or Socket Index:NiagaraInt32=&SelectIntFromArray(Direct Array Index=$User.SwingSectionIndex, Int Selection Array=DataInterface() {propertyValues {IntData [5,4,5,4]}})
    Second Bone or Socket Index:NiagaraInt32=&SelectIntFromArray(Direct Array Index=$User.SwingSectionIndex, Int Selection Array=DataInterface() {propertyValues {IntData [3,1,3,1]}})
   ) {module=FN_InitializeAnimTrailBoneIndecies002}
   &FN_InitializeAnimTrailBoneIndecies(
    AnimTrail Selection=1
    First Bone or Socket Index:NiagaraInt32=&SelectIntFromArray(Direct Array Index=$User.SwingSectionIndex, Int Selection Array=DataInterface() {propertyValues {IntData [5,4,5,4]}})
    Second Bone or Socket Index:NiagaraInt32=&SelectIntFromArray(Direct Array Index=$User.SwingSectionIndex, Int Selection Array=DataInterface() {propertyValues {IntData [2,0,2,0]}})
   ) {module=FN_InitializeAnimTrailBoneIndecies003}
   &FN_InitializeAnimTrailBoneIndecies(AnimTrail Selection=1, First Bone or Socket Index=2, Second Bone or Socket Index=3) {enabled=false, module=FN_InitializeAnimTrailBoneIndecies001}
   &FN_InitializeAnimTrail(Color=<rgba 0 0 0 0.15>, Lifetime=0.2)
   &AddVelocity(Velocity=<0 400 0>) {enabled=false}
  ]
  particleUpdateScript [
   &SolveForcesAndVelocity(Clamp Velocity=false, Limit Acceleration=false)
   &FN_UpdateAnimTrail() {module=FN_UpdateAnimTrail001}
   &DynamicMaterialParameters(
    Index 1 Param 3=&FloatFromCurve(
     FloatCurve=DataInterface() {propertyValues {Curve {keys [<at 0 0>, {arriveTangent=1.9814, tangentMode=RCTM_User, time=1, value=1}]}}}
     CurveIndex=&ModuloFloat(A=$Emitter.Age, B=1)
     Scale Curve=0
    )
    Index 1 Param 4=1
    Index 2 Param 1=10
    Index 2 Param 2=1
    Index 2 Param 3=0.5
    Index 2 Param 4=0.6
    Module.Index0Float0Write=true
    Module.Index0Float1Write=true
    Module.Index0Float2Write=true
    Module.Index0Float3Write=true
    Module.Index1Float0Write=true
    Module.Index1Float1Write=true
    Module.Index1Float2Write=true
    Module.Index1Float3Write=true
    Module.Index2Float0Write=true
    Module.Index2Float1Write=true
    Module.Index2Float2Write=true
    Module.Index2Float3Write=true
    Write Parameter Index 1=::NewEnumerator1
    Write Parameter Index 2=::NewEnumerator1
   )
   &ScaleColor(
    Scale Alpha=&MaskFloatBySpawnGroup(
     Masked Float=&FloatFromCurve(FloatCurve=DataInterface() {propertyValues {Curve {keys [{interpMode=RCIM_Cubic, tangentMode=RCTM_User, time=0.1, value=1}, {arriveTangent=-2.9197, tangentMode=RCTM_User, time=0.2, value=0}]}}})
     Passthrough Float=&FloatFromCurve(FloatCurve=DataInterface() {propertyValues {Curve {keys [{interpMode=RCIM_Cubic, tangentMode=RCTM_User, time=0, value=1}, {arriveTangent=-2.9197, tangentMode=RCTM_User, time=0.15, value=0}]}}})
     Spawn Group Mask=1
    )
    Scale RGB=&VectorFromFloat(Value=&FloatFromCurve(FloatCurve=DataInterface() {propertyValues {Curve {keys [<at 0 1>,<at 1 1>]}}}))
    ScaleA=true
    ScaleRGB=true
   )
  ]
  rendererClasses [&NiagaraRibbonRendererProperties]
  renderers [
   {rendererClass=&NiagaraRibbonRendererProperties}
   {rendererClass=&NiagaraRibbonRendererProperties, rendererIndex=1}
   {rendererClass=&NiagaraRibbonRendererProperties, rendererIndex=2}
  ]
 }
 MainTrail {
  emitterSpawnScript [&FN_RibbonIDOffsetUtility()]
  emitterUpdateScript [
   &EmitterState()
   &FN_SpawnAnimTrail(Number of Anim Trails=2)
  ]
  particleSpawnScript [
   &FN_InitializeAnimTrailBoneIndecies(
    First Bone or Socket Index:NiagaraInt32=&SelectIntFromArray(Direct Array Index=$User.SwingSectionIndex, Int Selection Array=DataInterface() {propertyValues {IntData [5,4,5,4]}})
    Second Bone or Socket Index:NiagaraInt32=&SelectIntFromArray(Direct Array Index=$User.SwingSectionIndex, Int Selection Array=DataInterface() {propertyValues {IntData [3,1,3,1]}})
   )
   &FN_InitializeAnimTrailBoneIndecies(
    AnimTrail Selection=1
    First Bone or Socket Index:NiagaraInt32=&SelectIntFromArray(Direct Array Index=$User.SwingSectionIndex, Int Selection Array=DataInterface() {propertyValues {IntData [5,4,5,4]}})
    Second Bone or Socket Index:NiagaraInt32=&SelectIntFromArray(Direct Array Index=$User.SwingSectionIndex, Int Selection Array=DataInterface() {propertyValues {IntData [2,0,2,0]}})
   ) {module=FN_InitializeAnimTrailBoneIndecies001}
   &FN_InitializeAnimTrail(Color=<rgba 1 1 1 0.5>, Lifetime=0.2)
   &AddVelocity(Velocity=<0 400 0>) {enabled=false}
   /SetVariables(
    Particles.Ribbon2Width:NiagaraFloat=&Multiply_Float(A=$Particles.RibbonWidth, B=&MaskFloatBySpawnGroup(Masked Float=1, Passthrough Float=0))
    Particles.RibbonWidth=&Multiply_Float(A=$Particles.RibbonWidth, B=&MaskFloatBySpawnGroup(Masked Float=0, Passthrough Float=1))
   ) {isSetParameters=true}
  ]
  particleUpdateScript [
   &SolveForcesAndVelocity(Clamp Velocity=false, Limit Acceleration=false)
   &FN_UpdateAnimTrail() {module=FN_UpdateAnimTrail001}
   &DynamicMaterialParameters(
    Index 1 Param 3=&FloatFromCurve(
     FloatCurve=DataInterface() {propertyValues {Curve {keys [<at 0 0>, {arriveTangent=1.9814, tangentMode=RCTM_User, time=1, value=1}]}}}
     CurveIndex=&ModuloFloat(A=$Emitter.Age, B=1)
     Scale Curve=0
    )
    Index 1 Param 4=1
    Index 2 Param 1=10
    Index 2 Param 2=1
    Index 2 Param 3=0.5
    Index 2 Param 4=0.6
    Module.Index0Float0Write=true
    Module.Index0Float1Write=true
    Module.Index0Float2Write=true
    Module.Index0Float3Write=true
    Module.Index1Float0Write=true
    Module.Index1Float1Write=true
    Module.Index1Float2Write=true
    Module.Index1Float3Write=true
    Module.Index2Float0Write=true
    Module.Index2Float1Write=true
    Module.Index2Float2Write=true
    Module.Index2Float3Write=true
    Write Parameter Index 1=::NewEnumerator1
    Write Parameter Index 2=::NewEnumerator1
   )
   &ScaleColor(
    Scale Alpha=&FloatFromCurve(
     FloatCurve=DataInterface() {propertyValues {Curve {keys [{interpMode=RCIM_Cubic, tangentMode=RCTM_User, time=0.1, value=1}, {arriveTangent=-2.9197, tangentMode=RCTM_User, time=0.2, value=0}]}}}
     Scale Curve=&Lerp_Float(
      A=1
      B=&FNTimeOfDayScalar(Night Scale=0.5, Morning Scale=1, Evening Scale=1)
      Alpha=$User.Variant
     )
    )
    Scale RGB=&MakeVectorFromLinearColorRGB(Color=$System.TOD)
    ScaleA=true
    ScaleRGB=true
   )
  ]
  rendererClasses [&NiagaraRibbonRendererProperties]
  renderers [
   {rendererClass=&NiagaraRibbonRendererProperties}
   {rendererClass=&NiagaraRibbonRendererProperties, rendererIndex=1}
   {rendererClass=&NiagaraRibbonRendererProperties, rendererIndex=2}
   {rendererClass=&NiagaraRibbonRendererProperties, rendererIndex=3}
   {rendererClass=&NiagaraRibbonRendererProperties, rendererIndex=4}
   {rendererClass=&NiagaraRibbonRendererProperties, rendererIndex=5}
  ]
 }
}
summary {
 systemName=NS_Pickaxe_AmpleSeller_Trail
 emitters [
  {emitterName=Far_CrossHatching, rendererClasses [&NiagaraRibbonRendererProperties]}
  {emitterName=MainTrail, rendererClasses [&NiagaraRibbonRendererProperties]}
  {emitterName=Close_CrossHatching, rendererClasses [&NiagaraRibbonRendererProperties]}
 ]
 userVariables [
  User.bIsFrontEnd {type=NiagaraBool}
  User.bIsFrontEndPreview {type=NiagaraBool}
  User.PhysicalSurface {type=NiagaraInt32}
  User.RibbonIDOffset {type=NiagaraInt32}
  User.SwingSectionIndex {type=NiagaraInt32}
  User.Variant {type=NiagaraFloat, defaultValue=1, description=""}
 ]
}
